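; Parent Blueprint for Skyfire environment lights, which sets color by rarity and hides the light below a chosen effects-quality tier.
; On BeginPlay it takes its rarity from the Skyfire EnvAsset manager if one exists, and binds a refresh to effects-quality changes.
; SetLightParameters, also run at construction, applies the mapped rarity color, light function material and quality-based visibility, with optional debug printing.
; All light changes run only where cosmetic events can execute.
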(event EventBeginPlay
  ;[0.1792 0.6974 1] BeginPlay
  (|CallParentFunction)
  (Networking|CanExecuteCosmeticEvents
    (:True
      (EventDispatchers|BindEventtoOnEffectsQualityChanged (Settings|GetFortGameUserSettings) (EventDispatchers|CreateEvent :!event SetLightParameters))
      ;;[0.1553 1 0.5413] Check with the Skyfire EnvAsset manager if we should update our rarity and lighting colors
      (bind _returnvalue (Actor|GetActorOfClass "/Game/Environments/Apollo/Sets/Skyfire/Blueprints/BP_Skyfire_EnvAsset_Manager.BP_Skyfire_EnvAsset_Manager_C"))
      (Utilities|IsValid _returnvalue
        (:Is%20Valid
          (bind _rarity (Class|BPSkyfireEnvAssetManager|GetRarityforEnvAsset _returnvalue self))
          (Variables|Default|SetSkyFireRarity _rarity))
        (:Is%20Not%20Valid
          _))
      (CallFunction|SetLightParameters))))

(fn LightQualityCheck
  (bind _qualitysetting (Variables|Default|GetQualitySetting))
  (bind _returnvalue (Utilities|Select :Index _qualitysetting :NewEnumerator0 0 :NewEnumerator1 1 :NewEnumerator2 2 :NewEnumerator3 3 :NewEnumerator4 4))
  (Networking|CanExecuteCosmeticEvents
    (:True
      (bind _returnvalue_2 (Settings|GetGameUserSettings))
      (Rendering|SetVisibility (Variables|Default|GetLightComponent) (>= (Settings|GetVisualEffectQuality _returnvalue_2) _returnvalue) false)
      (Development|PrintString (Utilities|String|Append "User Visual Effect Quality = " (Utilities|String|ToString(Integer) (Settings|GetVisualEffectQuality _returnvalue_2)) " BP Quality Setting = " (Utilities|String|ToString(Integer) _returnvalue) "  >= ? " (Utilities|String|ToString(Boolean) (>= (Settings|GetVisualEffectQuality _returnvalue_2) _returnvalue))) (Variables|Default|GetPrintLightQualityCheckDebugInfo) (Variables|Default|GetPrintLightQualityCheckDebugInfo) "(R=1.000000,G=0.000000,B=0.382202,A=1.000000)" 5.0))))

(fn SetLightFunctionMaterial
  (Networking|CanExecuteCosmeticEvents
    (:True
      (bind _lightfunctionmaterial (Variables|Default|GetLightFunctionMaterial)
        (:then
          (Rendering|Components|Light|SetLightFunctionMaterial (Variables|Default|GetLightComponent) _lightfunctionmaterial))
        (:else)))))

(fn SetLightParameters
  (CallFunction|SetRarityColorParameters)
  (CallFunction|SetLightFunctionMaterial)
  (CallFunction|LightQualityCheck))

(fn SetLightRarityColor
  (bind _skyfirerarity (Variables|Default|GetSkyFireRarity))
  (bind _skyfireraritylightcolors (Variables|Default|GetSkyfireRarityLightColors))
  (bind (_value _returnvalue) (Utilities|Map|Find _skyfireraritylightcolors _skyfirerarity))
  (Networking|CanExecuteCosmeticEvents
    (:True
      (if _returnvalue
        (Rendering|Components|Light|SetLightColor (Variables|Default|GetLightComponent) _value true)))))

(fn SetRarityColorParameters
  (CallFunction|SetLightRarityColor))

(fn ConstructionScript
  (|CallParentFunction)
  (CallFunction|SetLightParameters))
