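; Skyfire point light Blueprint for Athena Apollo environments, a thin child class that binds its light to inherited behavior.
; Its construction script assigns the Point Light component as the Light Component, then runs the parent construction script.
; BeginPlay only forwards to the parent, so all runtime lighting logic lives in the parent class.
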(event EventBeginPlay
  (|CallParentFunction))

(fn ConstructionScript
  (Variables|Default|SetLightComponent (Variables|Default|GetPointLight))
  (|CallParentFunction))
