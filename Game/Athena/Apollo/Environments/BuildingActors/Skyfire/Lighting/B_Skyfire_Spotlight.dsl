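; B_Skyfire_Spotlight is a Blueprint spotlight building actor in the Skyfire lighting set for the Apollo environment.
; Its construction script assigns the actor's spot light to the light component variable, then runs the parent construction script.
; BeginPlay only calls the parent, so any further lighting behavior comes from the parent class.
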(event EventBeginPlay
  (|CallParentFunction))

(fn ConstructionScript
  (Variables|Default|SetLightComponent (Variables|Default|GetSpotLight))
  (|CallParentFunction))
