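# Light function material for Skyfire environment lighting that projects animated caustics through a light.
# Two samples of the shield caustics texture scroll independently, each with a vector parameter whose red channel sets uniform tiling and blue channel sets equal pan speed on both axes.
# Their red channels are summed, divided, then offset by the Base scalar; the green and alpha parameter channels are unused.
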Material /Game/Athena/Apollo/Environments/BuildingActors/Skyfire/Lighting/M_Skyfire_Caustics
graph {
 properties {materialDomain=MD_LightFunction, pixelDepthOffsetMode=PDOM_Legacy, usedShadingModels=Default Lit}
 nodes [
  n1 = TextureCoordinate()
  n2 = VectorParameter() {parameterName=NoiseA_TilePan, defaultValue=<rgba 3 3 0.0875 -0.077>}
  n3 = AppendVector(A=@n2.R, B=@n2.R)
  n4 = TextureSample(UVs=Panner(Coordinate=Multiply(A=@n1, B=@n3), Speed=AppendVector(A=@n2.B, B=@n2.B)) {bFractionalPart=true}) {texture=&T_ShieldCaustics}
  n5 = VectorParameter() {parameterName=NoiseB_TilePan, defaultValue=<rgba -4 -4 0.075 0.0605>}
  n6 = AppendVector(A=@n5.R, B=@n5.R)
  n7 = TextureSample(UVs=Panner(Coordinate=Multiply(A=@n1, B=@n6), Speed=AppendVector(A=@n5.B, B=@n5.B)) {bFractionalPart=true}) {texture=&T_ShieldCaustics}
  Add(A=ScalarParameter() {defaultValue=0.5, parameterName=Base}, B=Divide(A=Add(A=@n4.R, B=@n7.R)))
 ]
}
